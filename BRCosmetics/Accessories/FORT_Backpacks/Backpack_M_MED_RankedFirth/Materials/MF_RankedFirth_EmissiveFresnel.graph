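# Material function for the RankedFirth backpack that adds a rank glow to a material's emissive, toggled by a static switch.
# The masked Fresnel input is raised to a power, multiplied by panning Perlin noise with noise-distorted UVs, then tinted and scaled.
# A RankedFirth glider switch projects the noise from pre-skinning local position and blends in a warped radial swirl around the center.
# The animated Fresnel-power and NoiseInfluence nodes are left unconnected.
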MaterialFunction /BRCosmetics/Accessories/FORT_Backpacks/Backpack_M_MED_RankedFirth/Materials/MF_RankedFirth_EmissiveFresnel
graph {nodes [
  n1 = GetMaterialAttributes(None=FunctionInput() {inputName=Material, inputType=FunctionInput_MaterialAttributes}) {attributeGetTypes [#1]}
  n2 = VectorParameter() {group=EmissiveRankEffect, parameterName=EmissiveColor, defaultValue=<rgba 0.6875 0.1451 0.0903 1>}
  n3 = FunctionInput() {inputName=Fresnel, inputType=FunctionInput_Scalar, sortPriority=2}
  n4 = ChannelMaskParameter(None=FunctionInput() {inputName=Mask, sortPriority=1}) {group=EmissiveRankEffect, parameterName=Mask Channel}
  n5 = ScalarParameter() {defaultValue=4, group=EmissiveRankEffect, parameterName=FresnelPower_01}
  n6 = ComponentMask(None=NamedRerouteUsage()) {b=true, g=true}
  n7 = Constant() {r=100}
  n8 = Multiply(A=StaticSwitchParameter(True=Divide(A=@n6, B=@n7) {constB=128}, False=TextureCoordinate()) {group=RankedFirth, parameterName=UseRankedFirthGlider}) {constB=0.5}
  n9 = Divide(A=ComponentMask(None=NamedRerouteUsage()) {b=true, r=true}, B=@n7) {constB=128}
  n10 = TextureCoordinate()
  n11 = VectorParameter() {group=EmissiveRankEffect, parameterName=NoiseSpeed}
  n12 = TextureSampleParameter2D(UVs=Panner(
    Coordinate=StaticSwitchParameter(True=@n9, False=@n10) {group=RankedFirth, parameterName=UseRankedFirthGlider}
    Speed=AppendVector(A=@n11.R, B=@n11.R)
   ) {speedX=0.2, speedY=0.12}) {parameterName=ThinFilmUVOffsetTexture, samplerSource=SSM_Wrap_WorldGroupSettings, texture=/Game/Global/Textures/Default/Noise/3xPerlin_M}
  n13 = Multiply(A=Add(A=@n12.G, B=@n12.G)) {constB=0.5}
  n14 = TextureSampleParameter2D(UVs=Panner(Coordinate=Add(A=@n8, B=@n13), Speed=AppendVector(A=@n11.B, B=@n11.B))) {parameterName=ThinFilmUVOffsetTexture, samplerSource=SSM_Wrap_WorldGroupSettings, texture=&3xPerlin_M}
  n15 = Multiply(A=Power(Base=Multiply(A=@n3, B=@n4.R), Exp=@n5) {constExponent=5}, B=@n14.R)
  n16 = StaticSwitchParameter(True=LinearInterpolate(A=NamedRerouteUsage(), B=@n15, Alpha=NamedRerouteUsage()), False=@n15) {group=RankedFirth, parameterName=UseRankedFirthGlider}
  n17 = Add(
   A=@n1.EmissiveColor
   B=Multiply(A=Multiply(A=@n2.RGB, B=@n16), B=ScalarParameter() {defaultValue=4, group=EmissiveRankEffect, parameterName=EmissiveStrength}) {constB=7}
  )
  FunctionOutput(None=SetMaterialAttributes!(
    MaterialAttributes=@n1.MaterialAttributes
    EmissiveColor=StaticSwitchParameter(True=@n17, False=@n1.EmissiveColor) {group=EmissiveRankEffect, parameterName=UseEmissiveRank, sortPriority=0}
   ) {attributeSetTypes [#1]})
  LinearInterpolate(B=@n14.R, Alpha=@n16) {constA=1}
  n18 = Sine(None=Multiply(A=Time(), B=ScalarParameter() {defaultValue=0.5, group=EmissiveRankEffect, parameterName=FresnelPowerChange_Frequency}))
  LinearInterpolate(
   A=@n5
   B=ScalarParameter() {defaultValue=5, group=EmissiveRankEffect, parameterName=FresnelPower_02}
   Alpha=Add(A=Multiply(A=@n18) {constB=0.5}) {constB=0.5}
  )
  NamedRerouteDeclaration(None=VertexInterpolator(VS=LocalPosition() {localOrigin=InstancePreSkinning}.XYZ).PS) {name=PS-LocPos, nodeColor=<rgba 1 0.3281 0 1>}
  n19 = Multiply(A=Time(), B=Constant2Vector() {desc=SPEED, g=0.1678, r=0.1})
  n20 = NamedRerouteUsage()
  n21 = MaterialFunctionCall(Vector or UVs=Divide(
    A=ComponentMask(None=@n20) {g=true, r=true}
    B=ScalarParameter() {defaultValue=300, group=RankedFirth, parameterName=GliderSwirlingNoiseSize}
   ) {constB=300}) {materialFunction=&VectorToRadialValue}
  n22 = Multiply(A=@n21.Radial Coordinates, B=Constant2Vector() {g=-2, r=2})
  n23 = TextureCoordinate()
  n24 = DDX(Value=@n23)
  n25 = DDY(Value=@n23)
  n26 = TextureSample!(
   UVs=Add(A=Frac(None=@n19), B=NamedRerouteDeclaration(None=@n22) {name=RadialUV, nodeColor=<rgba 0.3125 1 0 1>}) {constB=0}
   Tex
   "DDX(UVs)"=@n24
   "DDY(UVs)"=@n25
   Apply View MipBias
  ) {mipValueMode=TMVM_Derivative, samplerSource=SSM_Wrap_WorldGroupSettings, texture=&3xPerlin_M}
  n27 = ComponentMask(None=@n26.RGB) {g=true, r=true}
  n28 = ScalarParameter() {defaultValue=0.4, group=RankedFirth, parameterName=WarpingIntensity}
  n29 = Multiply(
   A=Time()
   B=ComponentMask(None=VectorParameter() {desc=sPEED, parameterName=SwirlSpeed, defaultValue=<rgba 0.2 0.15 0 1>}.RGB) {g=true, r=true}
  ) {constA=50}
  n30 = NamedRerouteDeclaration(None=ComponentMask(None=@n21.Radial Coordinates) {g=true}) {name=CenterGradient, nodeColor=<rgba 1 0.2344 0 1>}
  n31 = LinearInterpolate(B=ScalarParameter() {defaultValue=0.75, group=RankedFirth, parameterName=SwirlIntensity}, Alpha=Saturate(None=Multiply(A=@n30) {constB=3})) {constB=-0.75}
  n32 = Constant()
  n33 = Add(
   A=NamedRerouteUsage()
   B=Add(
    A=Frac(None=@n29)
    B=MaterialFunctionCall(X=@n31, Y=@n32) {materialFunction=&MakeFloat2}.Result
   )
  ) {constB=0}
  n34 = TextureSample(UVs=Add(A=Multiply(A=@n27, B=@n28), B=Frac(None=@n33)), "DDX(UVs)"=@n24, "DDY(UVs)"=@n25) {mipValueMode=TMVM_Derivative, samplerSource=SSM_Wrap_WorldGroupSettings, texture=&3xPerlin_M}
  n35 = Multiply(A=@n34.R, B=SmoothStep(Value=NamedRerouteUsage()) {constMax=0.3, constMin=0.08})
  n36 = ScalarParameter() {defaultValue=2, group=RankedFirth, parameterName=SwirlingNoisePower}
  NamedRerouteDeclaration(None=Multiply(A=Power(Base=@n35, Exp=@n36), B=ScalarParameter() {defaultValue=0.3, group=RankedFirth, parameterName=SwirlingNoiseIntensity})) {name=SwirlingNoise, nodeColor=<rgba 0 1 0.0859 1>}
  NamedRerouteDeclaration(None=Step(Y=@n30) {constX=0.05, constY=0.5}) {name=CenterMask, nodeColor=<rgba 0.8047 1 0 1>}
  Saturate(None=ScalarParameter() {defaultValue=1, group=EmissiveRankEffect, parameterName=NoiseInfluence})
 ]}
