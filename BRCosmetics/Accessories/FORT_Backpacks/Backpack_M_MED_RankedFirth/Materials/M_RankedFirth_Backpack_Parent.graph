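# Parent character material for the Ranked Firth backpack, layering ranked cosmetic effects between the CharacterParent_2023 PreFX and PostFX functions.
# Thin-film functions feed emissive Fresnel, an optional rank icon and a diamond effect.
# A switchable glider branch, reusing Season 32 glider assets, adjusts base color, metallic, specular and anisotropic tangent.
# Its color highlight samples a curve atlas using a sine of radial coordinates plus a sun-and-view half-vector dot normal.
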Material /BRCosmetics/Accessories/FORT_Backpacks/Backpack_M_MED_RankedFirth/Materials/M_RankedFirth_Backpack_Parent
graph {
 properties {
  bUseMaterialAttributes=true
  bUsedWithMeshDeformer=true
  bUsedWithNanite=true
  bUsedWithSkeletalMesh=true
  pixelDepthOffsetMode=PDOM_Legacy
  previewMesh=/BRCosmetics/Gadgets/Assets/VinderTech_GliderChute/Glider_Season_32/Meshes/Season_32
  shadingModel=MSM_FromMaterialExpression
  usedShadingModels=Default Lit | Subsurface | Clear Coat | Subsurface Profile | Cloth
 }
 nodes [
  n1 = NamedRerouteUsage()
  n2 = VectorParameter() {group=RankedFirth, parameterName=RadialHighlightTint, defaultValue=<rgba 1 1 1 1>}
  n3 = GetMaterialAttributes(None=NamedRerouteUsage()) {attributeGetTypes [#1]}
  n4 = CollectionParameter() {
   collection=&FortniteMaterialParameters
   parameterName=SunlightDirectionalVector
  }
  n5 = ActorPositionWS() {originType=CameraRelative}
  n6 = Add(A=Normalize(VectorInput=ComponentMask(None=@n4) {b=true, g=true, r=true}), B=Multiply(A=Normalize(VectorInput=@n5)) {constB=-1})
  n7 = Divide(A=ComponentMask(None=NamedRerouteUsage()) {g=true, r=true}, B=Constant() {r=133.5}) {constB=64}
  n8 = MaterialFunctionCall(Vector or UVs=@n7) {materialFunction=&VectorToRadialValue}
  n9 = Add(
   A=DotProduct(A=Transform(None=@n3.Normal), B=Normalize(VectorInput=@n6) {desc=Half vector})
   B=Add(A=ComponentMask(None=@n8.Radial Coordinates) {r=true}) {constB=0.065}
  )
  n10 = Constant() {r=4}
  n11 = ConstantBiasScale(None=Sine(None=Multiply(A=@n9, B=@n10) {constB=2}))
  n12 = CurveAtlasRowParameter(CurveTime=@n11) {
   atlas=/BRCosmetics/Gadgets/Assets/VinderTech_GliderChute/Glider_Season_32/Textures/RankGradients/CA_RankedFirth
   curve=/BRCosmetics/Gadgets/Assets/VinderTech_GliderChute/Glider_Season_32/Textures/RankGradients/C_RankedFirth
   group=RankedFirth
   parameterName=RadialColorCurve
  }
  n13 = Multiply(A=@n1, B=@n12)
  LinearInterpolate(A=Multiply(A=@n1, B=@n2.RGB), B=@n13, Alpha=Saturate(None=LinearInterpolate(A=@n11)))
  n14 = MaterialFunctionCall() {materialFunction=&MF_CharacterParent_2023_PreFX}
  n15 = NamedRerouteUsage()
  n16 = MaterialFunctionCall!(
   Material Attributes=MaterialFunctionCall!(
    Material Attributes=NamedRerouteDeclaration(None=@n14.Result) {name=MA_IN, nodeColor=<rgba 1 0 0.4219 1>}
    Fresnel=NamedRerouteUsage()
    Alpha=@n15
   ) {materialFunction=/Game/Characters/Player/Female/Medium/Bodies/F_MED_SkeleProbe/Materials/MF_ThinFilmDuplicate_TechArt}.Result
   Alpha=@n15
  ) {materialFunction=/Game/Global/Materials/Cosmetics/Ranked/MF_RankedThinFilm}
  n17 = MaterialFunctionCall!(
   Material=StaticSwitchParameter(True=NamedRerouteUsage(), False=@n16.Result) {group=RankedFirth, parameterName=UseRankedFirthGlider}
   Mask=@n15
   Fresnel=NamedRerouteUsage()
  ) {materialFunction=&/MF_RankedFirth_EmissiveFresnel}
  n18 = MaterialFunctionCall!(Material=@n17.Result, Alpha=@n15) {materialFunction=/BRCosmetics/Accessories/FORT_Backpacks/Backpack_M_MED_RankedDefender/Material/MF_RankedDefenderIcon}
  MaterialFunctionCall(
   Material=MaterialFunctionCall!(
    UV
    Material=StaticSwitchParameter(True=@n18.Result, False=@n17.Result) {group=RankIcon, parameterName=UseRankIcon}
    TechArtMask=NamedRerouteUsage()
   ) {materialFunction=/BRCosmetics/Weapons/FORT_Melee/Pickaxe_RankedDefender/Materials/MF_NewDiamond}.Material
   M=@n14.M
   Pre Skinned Local Position=NamedRerouteDeclaration(None=@n14.Pre skinned Local Position) {name=PS-LocalPos, nodeColor=<rgba 0 1 0.4609 1>}
   Distance Blend=@n14.M
   Fresnel=NamedRerouteUsage()
  ) {materialFunction=&MF_CharacterParent_2023_PostFX}
  Multiply(A=ScalarParameter() {defaultValue=0.5, group=RankedFirth, parameterName=DARKENING}, B=@n1) {constB=0.9}
  n19 = Subtract(A=@n14.Tech Art Mask, B=Multiply(A=NamedRerouteUsage(), B=ScalarParameter() {defaultValue=0.5, parameterName=VinylThinFilmMask_Intensity}))
  NamedRerouteDeclaration(None=StaticSwitchParameter(True=Saturate(None=@n19), False=@n14.Tech Art Mask) {group=RankedFirth, parameterName=UseVinylThinFilmMask}) {name=TechArtMask, nodeColor=<rgba 1 0.7969 0 1>}
  n20 = TextureObjectParameter() {
   group=RankedFirth
   parameterName=GliderAnisotropicTangent
   samplerType=SAMPLERTYPE_Masks
   texture=/BRCosmetics/Gadgets/Assets/VinderTech_GliderChute/Glider_Season_32/Textures/T_Glider_Season_32_Tangent
  }
  n21 = TextureSample(UVs=Add(A=@n7, B=Constant2Vector() {g=0.5, r=0.5}), Tex=@n20) {
   samplerType=SAMPLERTYPE_Masks
   texture=&T_Glider_Season_32_Tangent
  }
  NamedRerouteDeclaration(None=LinearInterpolate(
    A=Constant3Vector() {constant=<rgba -1 -1 -1 0>}
    Alpha=NamedRerouteDeclaration(None=@n21.RGB) {name=TangentRGB, nodeColor=<rgba 0 0.6953 1 1>}
   )) {name=Glider_Tangent, nodeColor=<rgba 0.9688 1 0 1>}
  n22 = TextureSample(Tex=@n20) {
   samplerType=SAMPLERTYPE_Masks
   texture=&T_Glider_Season_32_Tangent
  }
  NamedRerouteDeclaration(None=LinearInterpolate(
    B=ScalarParameter() {defaultValue=0.5, group=RankedFirth, parameterName=GliderAnisotropyMaxWeight, sliderMax=1}
    Alpha=NamedRerouteDeclaration(None=@n22.A) {name=VinylMask, nodeColor=<rgba 0 1 1 1>}
   )) {name=Glider_Aniso, nodeColor=<rgba 0.7109 0 1 1>}
  n23 = Multiply(A=@n13, B=@n2.RGB)
  n24 = ScalarParameter() {defaultValue=0.5, group=RankedFirth, parameterName=ThinFilmColorBlend}
  NamedRerouteDeclaration(None=LinearInterpolate(A=@n1, B=LinearInterpolate(A=@n1, B=@n23, Alpha=@n24), Alpha=NamedRerouteUsage())) {name=Glider_BaseColor, nodeColor=<rgba 0 1 0.3906 1>}
  n25 = GetMaterialAttributes(None=@n16.Result) {attributeGetTypes [#2]}
  NamedRerouteDeclaration(None=@n25.Base Color) {name=AfterThinFilm_BaseColor, nodeColor=<rgba 1 0.2812 0 1>}
  n26 = SetMaterialAttributes!(
   MaterialAttributes=@n25.MaterialAttributes
   Tangent=NamedRerouteUsage()
   Anisotropy=NamedRerouteUsage()
   Specular=NamedRerouteUsage()
   Metallic=NamedRerouteUsage()
   Base Color=NamedRerouteUsage()
  ) {attributeSetTypes [
    #3
    #4
    #5
    #6
    #2
   ]}
  NamedRerouteDeclaration(None=@n26) {name=MA_Glider, nodeColor=<rgba 0.6641 1 0 1>}
  NamedRerouteDeclaration(None=@n11) {name=Debug, nodeColor=<rgba 0 0.5078 1 1>}
  NamedRerouteDeclaration(None=@n14.Fresnel) {name=Fresnel, nodeColor=<rgba 1 0 0.1406 1>}
  n27 = GetMaterialAttributes(None=NamedRerouteUsage()) {attributeGetTypes [
    #5
    #3
    #7
    #6
   ]}
  n28 = Add(A=@n27.Metallic, B=ScalarParameter() {defaultValue=0.5, group=RankedFirth, parameterName=MetallicBias})
  NamedRerouteDeclaration(None=LinearInterpolate(A=@n27.Metallic, B=Saturate(None=@n28), Alpha=NamedRerouteUsage())) {name=Glider_Metallic, nodeColor=<rgba 1 0 0.3281 1>}
  n29 = Multiply(A=@n27.Specular, B=ScalarParameter() {defaultValue=3, group=RankedFirth, parameterName=SpecularMulti})
  NamedRerouteDeclaration(None=LinearInterpolate(A=@n27.Specular, B=Saturate(None=@n29), Alpha=NamedRerouteUsage())) {name=Glider_Specular, nodeColor=<rgba 0 0.6484 1 1>}
  SetMaterialAttributes!(MaterialAttributes, Base Color=NamedRerouteUsage()) {attributeSetTypes [#2]}
 ]
}
